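# Test Default Lit material for a monster mesh, built from a wind-style wobble and a bubbly tinted pattern.
# Two SimpleGrassWind setups with different intensity, weight and speed are lerped by a time-driven sine, producing an oscillating offset.
# A purple Color parameter feeds MultiplyAdd, with two differently scaled world-aligned Bubbles normal-map samples multiplied as its additive term.
# The graph shows no connections to material output pins.
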Material /BRMapCh6/Maps/Test/Chris/Meshes/Monsters/Monster_B_TstMaterial
graph {
 properties {bAutomaticallySetUsageInEditor=false, usedShadingModels=Default Lit}
 nodes [
  n1 = Constant() {r=1}
  n2 = Time()
  LinearInterpolate(
   A=MaterialFunctionCall!(WindIntensity=Constant() {r=0.5}, WindWeight=Constant() {r=2}, WindSpeed=Constant() {r=0.25}, AdditionalWPO=Constant()) {materialFunction=/Engine/Functions/Engine_MaterialFunctions01/WorldPositionOffset/SimpleGrassWind}.Result
   B=MaterialFunctionCall(WindIntensity=Constant() {r=3}, WindWeight=Constant() {r=0.15}, WindSpeed=@n1, AdditionalWPO=@n1) {materialFunction=&SimpleGrassWind}.Result
   Alpha=Sine(None=Divide(A=@n2) {constB=3})
  )
  MaterialFunctionCall(
   TextureObject=TextureObject() {samplerType=SAMPLERTYPE_Normal, texture=/Game/Items/Art_noLOD/Textures/Bubbles}
   TextureSize=Constant3Vector() {constant=<rgba 1100 1100 1100 0>}
  ) {materialFunction=&WorldAlignedTexture}
  n3 = TextureObject() {samplerType=SAMPLERTYPE_Normal, texture=&Bubbles}
  n4 = Constant3Vector() {constant=<rgba 512 512 512 0>}
  n5 = TextureObject() {samplerType=SAMPLERTYPE_Normal, texture=&Bubbles}
  n6 = Constant3Vector() {constant=<rgba 1100 1100 1100 0>}
  MaterialFunctionCall!(
   Base=VectorParameter() {parameterName=Color, defaultValue=<rgba 0.0938 0.0036 0.0737 1>}.RGB
   Add=Multiply(
    A=MaterialFunctionCall(TextureObject=@n3, TextureSize=@n4) {materialFunction=&WorldAlignedTexture}.XYZ Texture
    B=MaterialFunctionCall(TextureObject=@n5, TextureSize=@n6) {materialFunction=&WorldAlignedTexture}.XYZ Texture
   )
  ) {materialFunction=/Engine/Functions/Engine_MaterialFunctions02/Math/MultiplyAdd}
 ]
}
