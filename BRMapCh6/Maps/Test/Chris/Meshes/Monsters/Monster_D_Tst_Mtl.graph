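# Test material in the Chris test map's Monsters mesh folder, apparently for monster D, using Default Lit shading with a teal default color.
# It exposes Color, Glow and Roughness parameters, with no textures.
# Color feeds two multiplies: one with a constant, the other scaled by Glow, which defaults to zero.
# The listing does not show which material outputs these nodes drive.
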Material /BRMapCh6/Maps/Test/Chris/Meshes/Monsters/Monster_D_Tst_Mtl
graph {
 properties {bAutomaticallySetUsageInEditor=false, usedShadingModels=Default Lit}
 nodes [
  n1 = VectorParameter() {parameterName=Color, defaultValue=<rgba 0.0061 0.4844 0.4287 1>}
  Multiply(A=@n1.RGB)
  Multiply(A=@n1.RGB, B=ScalarParameter() {parameterName=Glow}) {constB=0}
  ScalarParameter() {defaultValue=0.3, parameterName=Roughness}
 ]
}
